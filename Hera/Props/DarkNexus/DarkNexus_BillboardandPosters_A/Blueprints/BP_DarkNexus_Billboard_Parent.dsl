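; Parent Blueprint for Dark Nexus billboard and poster props, handling fixture tinting and exposed-asset registration.
; Its construction script finds an InstancedStaticMeshComponent and, if valid, applies FixtureColor to it as the secondary mesh.
; Exposed asset provider overrides add version 1 and collect the resolved building texture data and swatch data asset as indirect asset and class references, respectively.
; BeginPlay, overlap and tick events contain no active logic.
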(event EventBeginPlay
  (|CallParentFunction) ; This node is disabled and will not be called.\nDrag off pins to build functionality.
)

(event Collision|EventActorBeginOverlap
  (params (OtherActor :type Actor%20Object%20Reference)))

(event EventTick
  (params (DeltaSeconds :type Float%20%28single-precision%29)))

(fn BuildExposedAssetProviderVersion(Implementation)
  (params (Version :type Fort%20Exposed%20Asset%20Provider%20Version%20Structure%20%28by%20ref%29))
  (|CallParentFunction Version)
  (ExposedAssetProvider|AddVersion Version 1)
  (return))

(fn CollectAssetReferencestoAllow(Implementation)
  (params (Collector :type Fort%20Exposed%20Asset%20Reference%20Collector%20Structure%20%28by%20ref%29))
  (|CallParentFunction Collector)
  (ExposedAssetProvider|CollectIndirectAssetReference Collector (Utilities|ResolveSoftReference (LocalVariables|GetBuildingTextureData)))
  (return))

(fn CollectClassReferencestoAllow(Implementation)
  (params (Collector :type Fort%20Exposed%20Class%20Reference%20Collector%20Structure%20%28by%20ref%29))
  (|CallParentFunction Collector)
  (ExposedAssetProvider|CollectIndirectClassReference Collector (Utilities|ResolveSoftReference (Variables|RandomSwatch|GetSwatchDataAsset)))
  (return))

(fn ConstructionScript
  (bind _returnvalue (Actor|GetComponentbyClass "/Script/Engine.InstancedStaticMeshComponent"))
  (|CallParentFunction)
  (if (Utilities|IsValid _returnvalue :!class /Script/Engine.KismetSystemLibrary)
    (Default|SetColoronSecondaryMesh (Variables|Default|GetFixtureColor) _returnvalue)))
